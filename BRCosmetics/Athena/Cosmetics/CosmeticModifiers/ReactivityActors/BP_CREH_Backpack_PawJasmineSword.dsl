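; Cosmetic reactivity effects handler for the PawJasmineSword backpack, animating its materials in response to reaction events.
; After customization it switches reactivity off on the backpack mesh.
; Effect tag 2 routes to ReactivityEvent1, always as Start, which stores a blackboard intensity, enables reactivity and plays a timeline driving scrolling and swoosh parameters.
; Reactivity switches off when the timeline finishes; the stop branch is unreachable through this routing.
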(event EventNotifyReactivityEffect
  (params (EffectTag :type Gameplay%20Tag%20Structure) (EffectStage :type EReactivityEffectStage%20Enum :default "Start") (ReactionEventContext :type Cosmetic%20Reaction%20Event%20Context%20Structure%20%28by%20ref%29))
  ; Process the Notify Reactivity Event received and calls the custom event on the ReapGraph
  (CallFunction|RouteReactivityEvent EffectTag EffectStage ReactionEventContext))

(event EventOnPostOwnerFinishedCustomization
  ; Initialization when the customization pass is completed
  (Rendering|Material|SetScalarParameterValueonMaterials (Customization|GetSkeletalMeshforPartType (CosmeticModifierContext|GetAssociatedPlayerPawn (Variables|CosmeticReactivityEffectsHandler|GetCosmeticModifierContext)) "Backpack" :!class /Script/FortniteGame.FortPlayerPawn) "Enable Reactivity?" 0.0))

(event Custom|ReactivityEvent1
  (params (EffectStage :type EReactivityEffectStage%20Enum :default "Start") (EventContext :type Cosmetic%20Reaction%20Event%20Context%20Structure))
  (bind _targetcosmeticmesh (Variables|Default|GetTargetCosmeticMesh))
  ; ReactivityEffect_1 Implementation
  (Variables|Default|SetEffectIntensity (CosmeticBlackboard|GetCosmeticBlackboardFloat (Utilities|Struct|BreakCosmeticReactionEventContext EventContext) "ReactionEffect1.EffectIntensity"))
  (Variables|Default|SetTargetCosmeticMesh (Customization|GetSkeletalMeshforPartType (CosmeticModifierContext|GetAssociatedPlayerPawn (Variables|CosmeticReactivityEffectsHandler|GetCosmeticModifierContext)) "Backpack" :!class /Script/FortniteGame.FortPlayerPawn))
  (if (Utilities|Enum|Equal(Enum) EffectStage "Start")
    ;; On Start Reaction Effect 2
    (Rendering|Material|SetScalarParameterValueonMaterials (Variables|Default|GetTargetCosmeticMesh) "Enable Reactivity?" 1.0)
    (bind (_direction _time_step _opacity) (Timeline|Timeline PlayFromStart)
      (:Update
        (Rendering|Material|SetScalarParameterValueonMaterials _targetcosmeticmesh "Scrolling - Time Step" _time_step)
        (Rendering|Material|SetScalarParameterValueonMaterials _targetcosmeticmesh "Scrolling - Intensity" _opacity)
        (Rendering|Material|SetScalarParameterValueonMaterials _targetcosmeticmesh "Swoosh Intensity" _opacity))
      (:Finished
        (Rendering|Material|SetScalarParameterValueonMaterials _targetcosmeticmesh "Enable Reactivity?" 0.0)))
    (else
      ;; On End Reaction Effect 2
      (Components|Timeline|Stop (Variables|BP_CREH_Backpack_PawJasmineSword|GetTimeline))
      (Rendering|Material|SetScalarParameterValueonMaterials (Variables|Default|GetTargetCosmeticMesh) "Enable Reactivity?" 0.0))))

(fn RouteReactivityEvent
  (params (ReactivityEffectTag :type Gameplay%20Tag%20Structure) (EffectStage :type EReactivityEffectStage%20Enum :default "Start") (EventContext :type Cosmetic%20Reaction%20Event%20Context%20Structure%20%28by%20ref%29))
  (Utilities|FlowControl|Sequence
    (:then_0)
    (:then_1
      (if (== (Variables|GetReactivityEffectTag) "(TagName=\"CosmeticModifier.Effect.2\")")
        (CallFunction|ReactivityEvent1 "Start" (Variables|GetEventContext))))))

(fn ConstructionScript)
